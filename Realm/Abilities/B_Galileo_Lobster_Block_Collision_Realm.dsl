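; Realm variant of the Galileo Lobster block collision, which handles blocking and deflection feedback for the blade.
; Hit events cast the other actor to the FloppingRabbit projectile and pass it, its owner, the hit and the blade's hot/cool colors to ExecuteBlockGameplayCue.
; That override checks the causer's team affiliation and calls RealmIsDeflecting on the owning Realm pawn for neutral or hostile sources, then returns the parent implementation's result.
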(event EventBeginPlay
  (|CallParentFunction) ; This node is disabled and will not be called.\nDrag off pins to build functionality.
)

(event Collision|EventActorBeginOverlap
  (params (OtherActor :type Actor%20Object%20Reference)))

(event EventTick
  (params (DeltaSeconds :type Float%20%28single-precision%29)))

(event Collision|EventHit
  (params (MyComp :type Primitive%20Component%20Object%20Reference) (Other :type Actor%20Object%20Reference) (OtherComp :type Primitive%20Component%20Object%20Reference) (bSelfMoved :type Boolean) (HitLocation :type Vector :default "0, 0, 0") (HitNormal :type Vector :default "0, 0, 0") (NormalImpulse :type Vector :default "0, 0, 0") (Hit :type Hit%20Result%20Structure%20%28by%20ref%29))
  (bind _asprj_athena_flopping_rabbit (Utilities|Casting|CastToPrj_Athena_FloppingRabbit Other))
  (CallFunction|ExecuteBlockGameplayCue :Damage 0.0 :DamageCauser _asprj_athena_flopping_rabbit :Instigator (Actor|GetOwner _asprj_athena_flopping_rabbit) :HitResult Hit :Color%20Hot (Math|Conversions|ToVector(LinearColor) (Variables|/Galileo/Items/Lobster/Actors/B_Athena_Galileo_Lobster_Parent.B_Athena_Galileo_Lobster_Parent_C|GetBlade%20Color%20Hot (Variables|Default|GetBP_GalileoLobster))) :Color%20Cool (Math|Conversions|ToVector(LinearColor) (Variables|/Galileo/Items/Lobster/Actors/B_Athena_Galileo_Lobster_Parent.B_Athena_Galileo_Lobster_Parent_C|GetBlade%20Color%20Cool (Variables|Default|GetBP_GalileoLobster)))))

(fn ExecuteBlockGameplayCue
  (params (Damage :type Float%20%28double-precision%29) (DamageCauser :type Actor%20Object%20Reference) (EffectContext :type Gameplay%20Effect%20Context%20Handle%20Structure) (Instigator :type Actor%20Object%20Reference) (HitResult :type Hit%20Result%20Structure) (Color%20Hot :type Vector :default "0, 0, 0") (Color%20Cool :type Vector :default "0, 0, 0"))
  (Utilities|FlowControl|Sequence
    (:then_0
      (switch Utilities|FlowControl|Switch|SwitchonEFortTeamAffiliation (Team|GetTeamAffiliation (Variables|Default|GetPlayerPawn) DamageCauser)
        (:Neutral
          _)
        (:Hostile
          _))
      (bind _asbp_player_pawn_realm (Utilities|Casting|CastToBP_PlayerPawn_Realm (Actor|GetOwner)))
      (Class|BPPlayerPawnRealm|RealmIsDeflecting _asbp_player_pawn_realm))
    (:then_1
      (bind _wasgalileobun (|CallParentFunction Damage DamageCauser EffectContext Instigator HitResult))
      (return _wasgalileobun))))

(fn ConstructionScript
  (|CallParentFunction))
