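; Realm deflect ability that makes an AI bot start and stop its block abilities, selected by gameplay tag, in response to threat and damage.
; On activation it binds to pawn damage and bot alert changes, triggering block when Threatened and cancelling it when Unaware.
; Damage triggers block unless Vader-usage or Lobster-unusable tags are present, and gaining the latter cancels it.
; Ending the ability unbinds damage and cancels block.
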(event Ability|EventFailedToActivatePassiveAbility
  (params (ActorInfo :type Gameplay%20Ability%20Actor%20Info%20Structure)))

(event Ability|EventActivateAbility
  (bind _ownerplayerpawn (Variables|Default|GetOwnerPlayerPawn))
  (Ability|CommitAbility)
  (bind _outfortplayerpawn (CallFunction|GetAvatarasValidFortPlayerPawn)
    (:ValidFortPlayerPawn
      (bind _output_get (Variables|Default|SetOwnerPlayerPawn _outfortplayerpawn))
      (bind _output_get_2 (Variables|Default|SetAbilityComponent (Ability|GetFortAbilitySystemComponent _output_get)))
      (Utilities|IsValid _output_get_2
        (:Is%20Valid
          (EventDispatchers|BindEventtoOnDamaged _ownerplayerpawn (delegate Custom|OnDamage))
          (bind _asbp_phoebe_controller_non_participant (Utilities|Casting|CastToBP_PhoebeController_NonParticipant (Pawn|GetController _ownerplayerpawn)))
          (AlertLevel|BindEventtoOnAlertLevelChangedEvent _asbp_phoebe_controller_non_participant (delegate Custom|RealmAlertLevelChanged))
          (Ability|Tasks|WaitGameplayTagRemove (Variables|Default|GetVaderAbilityUsageTag)
            (:then
              (Ability|Tasks|WaitGameplayTagRemove (Variables|Default|GetLobsterUnusableTag)
                (:then
                  (Ability|Tasks|WaitGameplayTagAdd (Variables|Default|GetLobsterUnusableTag)
                    (:Added
                      (Tags|CancelAbilitiesWithTags (Variables|Default|GetAbilityComponent) (Variables|Default|GetAbilityBlockGameplayTags)))))
                (:Removed
                  _)))
            (:Removed
              _))
          (if (Utilities|Enum|Equal(Enum) (Variables|/Script/FortniteAI.FortAthenaAIBotController|GetCurrentAlertLevel _asbp_phoebe_controller_non_participant) "Threatened")
            (Utilities|FlowControl|Delay 0.2)
            (Abilities|TryActivateAbilitiesbyTag (Variables|Default|GetAbilityComponent) (Variables|Default|GetAbilityBlockGameplayTags) true)))
        (:Is%20Not%20Valid
          _)))
    (:AvatarCastFailed
      _))
  (Ability|EndAbility))

(event Custom|RealmAlertLevelChanged
  (params (BotController :type Fort%20Athena%20AIBot%20Controller%20Object%20Reference) (OldAlertLevel :type EAlertLevel%20Enum :default "Unaware") (NewAlertLevel :type EAlertLevel%20Enum :default "Unaware"))
  (if (Utilities|Enum|Equal(Enum) NewAlertLevel "Threatened")
    (Abilities|TryActivateAbilitiesbyTag (Variables|Default|GetAbilityComponent) (Variables|Default|GetAbilityBlockGameplayTags) true)
    (elif (Utilities|Enum|Equal(Enum) NewAlertLevel "Unaware")
      (Tags|CancelAbilitiesWithTags (Variables|Default|GetAbilityComponent) (Variables|Default|GetAbilityBlockGameplayTags)))))

(event Custom|OnDamage
  (params (DamagedActor :type Actor%20Object%20Reference) (Damage :type Float%20%28single-precision%29) (InstigatedBy :type Controller%20Object%20Reference) (DamageCauser :type Actor%20Object%20Reference) (HitLocation :type Vector) (FHitComponent :type Primitive%20Component%20Object%20Reference) (BoneName :type Name) (Momentum :type Vector))
  (if (not (GameplayTags|HasAnyMatchingGameplayTags (Variables|Default|GetOwnerPlayerPawn) (GameplayTags|MakeGameplayTagContainerfromArray (Utilities|Array|MakeArray (Variables|Default|GetVaderAbilityUsageTag) (Variables|Default|GetLobsterUnusableTag)))))
    (Abilities|TryActivateAbilitiesbyTag (Variables|Default|GetAbilityComponent) (Variables|Default|GetAbilityBlockGameplayTags) true)))

(event Ability|EventOnEndAbility
  (params (bWasCancelled :type Boolean))
  (bind _outfortplayerpawn (CallFunction|GetAvatarasValidFortPlayerPawn)
    (:ValidFortPlayerPawn
      (EventDispatchers|UnbindEventfromOnDamaged _outfortplayerpawn (delegate Custom|OnDamage))
      (Tags|CancelAbilitiesWithTags (Variables|Default|GetAbilityComponent) (Variables|Default|GetAbilityBlockGameplayTags)))))
