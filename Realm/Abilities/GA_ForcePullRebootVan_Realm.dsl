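; Child force-pull ability in Realm that pulls a targeted pawn standing on a Reboot Van.
; Activation requires a valid target within max distance and a short downward trace from it hitting BGA_Athena_SCMachine.
; The destination comes from the parent, but if it lies under a threshold distance from the target, it becomes a fixed-distance offset opposite the avatar's forward vector.
; The high-arc check always returns true.
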(event Ability|EventActivateAbility
  (|CallParentFunction) ; This node is disabled and will not be called.\nDrag off pins to build functionality.
)

(event Ability|EventOnEndAbility
  (params (bWasCancelled :type Boolean))
  (|CallParentFunction bWasCancelled) ; This node is disabled and will not be called.\nDrag off pins to build functionality.
)

(fn CalculateShouldUseHighArc
  (params (StartLocation :type Vector :default "0, 0, 0") (EndLocation :type Vector :default "0, 0, 0"))
  (return true))

(fn GetEndForcePullDestination
  (bind _enddestination (|CallParentFunction))
  (if (>= (Math|Vector|Distance(Vector) _enddestination (Transformation|GetActorLocation (Variables|InternalData|GetFort%20Pawn%20Target))) (ScalableFloat|GetValueatLevel (LocalVariables|GetAcceptablePullThreshold)))
    (return _enddestination)
    (else
      (return (+ (Transformation|GetActorLocation (Variables|InternalData|GetFort%20Pawn%20Target)) (* (* (Transformation|GetActorForwardVector (Ability|GetAvatarActorfromActorInfo)) -1.0) (ScalableFloat|GetValueatLevel (LocalVariables|GetRebootVanPullDistance))))))))

(fn CanActivateAbility
  (params (ActorInfo :type Gameplay%20Ability%20Actor%20Info%20Structure) (Handle :type Gameplay%20Ability%20Spec%20Handle%20Structure))
  (bind _fortpawntarget (CallFunction|GetCurrentPawnTarget))
  (Utilities|IsValid _fortpawntarget
    (:Is%20Valid
      (if (Math|Float|InRange(Float) (Math|Vector|Distance(Vector) (Transformation|GetActorLocation _fortpawntarget) (Transformation|GetActorLocation (CallFunction|GetAvatarasFortPlayerPawn))) 0.0 (ScalableFloat|GetValueatLevel (LocalVariables|GetMaxDistance)))
        (bind (_outhit _returnvalue) (Collision|LineTraceByChannel (Transformation|GetActorLocation _fortpawntarget) (- (Transformation|GetActorLocation _fortpawntarget) "0,0,150.000000") "TraceTypeQuery7" false (Utilities|Array|MakeArray _fortpawntarget)))
        (if _returnvalue
          (Utilities|Casting|CastToBGA_Athena_SCMachine (.HitActor (Collision|BreakHitResult _outhit))
            (:then
              (return 0 true))
            (:CastFailed
              _))
          (else
            _))
        (return 0 false)
        (else
          _)))
    (:Is%20Not%20Valid
      _))
  (return 0 false))
