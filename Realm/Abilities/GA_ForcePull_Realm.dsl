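; Gameplay ability for AI-controlled Vader's Force Pull, launching the blackboard target toward a point in front of him.
; Activation requires range, clear line of sight (Reboot Van excepted), no surface swimming and no knockback immunity.
; A montage notify solves projectile velocity, choosing high arcs by angle or van footing, then launches the target or its ridden animal.
; It applies fall-damage immunity and a looping cue, and continuous targeting ends the ability early on proximity.
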(event Ability|EventActivateAbility
  (bind _fort_pawn_target (Variables|InternalData|GetFort%20Pawn%20Target))
  (bind _returnvalue (Actor|GetAttachParentActor _fort_pawn_target))
  (bind _fortpawntarget (CallFunction|GetCurrentPawnTarget))
  (bind _montageinfo_montagetoplay (Utilities|Struct|BreakFortGameplayAbilityMontageInfo (Variables|GetMontageInfo_MontageToPlay)))
  (Ability|CommitAbility)
  (Utilities|IsValid _fortpawntarget
    (:Is%20Valid
      (Variables|InternalData|SetFort%20Pawn%20Target _fortpawntarget)
      (Ability|Tasks|PlayMontageAndWaitNotify "None" _montageinfo_montagetoplay
        (:OnCompleted
          _)
        (:OnInterrupted
          _)
        (:OnCancelled
          _)
        (:OnNotifyBegin
          (bind (_tossvelocity _returnvalue_2) (Game|SuggestProjectileVelocity (+ (Transformation|GetActorLocation (Variables|InternalData|GetFort%20Pawn%20Target)) (Variables|/Script/Engine.MovementComponent|GetVelocity (Variables|/Script/Engine.Character|GetCharacterMovement (Variables|InternalData|GetFort%20Pawn%20Target)))) (CallFunction|GetEndForcePullDestination) (ScalableFloat|GetValueatLevel (Variables|Default|GetLaunchSpeed)) 0.0 "DoNotTrace" 0.0 (CallFunction|CalculateShouldUseHighArc (Transformation|GetActorLocation (CallFunction|GetAvatarasFortPlayerPawn)) (Transformation|GetActorLocation (Variables|InternalData|GetFort%20Pawn%20Target)))))
          (Variables|InternalData|SetToss%20Velocity _tossvelocity)
          (if _returnvalue_2
            (Components|Movement|StopMovementImmediately (Variables|/Script/Engine.Character|GetCharacterMovement (Variables|InternalData|GetFort%20Pawn%20Target)))
            (if (GameplayTags|HasAnyMatchingGameplayTags (Variables|InternalData|GetFort%20Pawn%20Target) (Variables|Default|GetTagsToValidateRiding))
              (Utilities|IsValid _returnvalue
                (:Is%20Valid
                  (bind _asfort_aipawn (Utilities|Casting|CastToFortAIPawn _returnvalue)
                    (:then
                      (Character|LaunchCharacter _asfort_aipawn (Variables|InternalData|GetToss%20Velocity) true true) ; Launch Animal Instead
                    )
                    (:CastFailed
                      (Character|LaunchCharacter _fort_pawn_target (Variables|InternalData|GetToss%20Velocity) true true :!label _label)))
                  (GameplayEffects|ApplyGameplayEffectToTarget :self (Ability|GetAbilitySystemComponent (Variables|InternalData|GetFort%20Pawn%20Target)) :GameplayEffectClass (Variables|Default|GetGE_FallDamageImmunity) :Target (Ability|GetAbilitySystemComponent (Variables|InternalData|GetFort%20Pawn%20Target)) :Level 0.0)
                  (GameplayCue|AddGameplayCueOnActor(Looping) :Target (Variables|InternalData|GetFort%20Pawn%20Target) :Parameters (Ability|GameplayCue|MakeGameplayCueParameters 0.0))
                  (CallFunction|ValidateTargetProximity))
                (:Is%20Not%20Valid
                  _)))
            (goto _label)
            (else
              (Ability|EndAbility :!label _label_2)))))
      ;[0.9479 0.9479 0.9479 0.5] Allows ability to cancel properly
      (Utilities|FlowControl|DelayUntilNextTick)
      (Ability|EndAbility))
    (:Is%20Not%20Valid
      _))
  (goto _label_2))

(event Ability|EventCommitExecute
  (|CallParentFunction)
  (Ability|SendGameplayEvent "(TagName=\"Ability.Realm.ForcePull.Commit\")"))

(event Custom|ValidateTargetProximity
  ;[1 1 1] Check to see when our target is close so we can stop the ability sooner in case the target gets close to Vader before his animation is over.
  (bind (_asynctaskproxy _targetdata) (Ability|Tasks|DoTargetSelectionContinuous "None" "(TagName=\"Targeting.ApplicationTag.0\")")
    (:Targeted
      (for _array_element (Ability|TargetData|GetAllActorsfromTargetData _targetdata)
        (if (== _array_element (Variables|InternalData|GetFort%20Pawn%20Target))
          (Ability|EndAbility))))))

(event Ability|EventOnEndAbility
  (params (bWasCancelled :type Boolean)))

(fn CalculateShouldUseHighArc
  (params (StartLocation :type Vector) (EndLocation :type Vector))
  (bind (_endlocation_x _endlocation_y _endlocation_z) (Math|Vector|BreakVector (Variables|GetEndLocation_X)))
  (bind (_startlocation_x _startlocation_y _startlocation_z) (Math|Vector|BreakVector (Variables|GetStartLocation_X)))
  (if (and (> (Class|AbilityFunctions|GetAngleBetweenTwoVectors (Transformation|GetActorForwardVector (Ability|GetAvatarActorfromActorInfo)) (Math|Vector|Normalize (- (Variables|GetEndLocation) (Variables|GetStartLocation)))) (ScalableFloat|GetValueatLevel (Variables|Default|GetHighArcMaxAngle))) (< _startlocation_z _endlocation_z))
    (return true)
    (else
      ;[1 1 1] Somewhat of a Hax for the Reboot Van - Make sure Vader can properly pull players from the Van
      (bind (_outhit _returnvalue) (Collision|LineTraceByChannel (Transformation|GetActorLocation (Variables|InternalData|GetFort%20Pawn%20Target)) (- (Transformation|GetActorLocation (Variables|InternalData|GetFort%20Pawn%20Target)) "0,0,150.000000") "TraceTypeQuery7" false (Utilities|Array|MakeArray (Variables|InternalData|GetFort%20Pawn%20Target))))
      (if _returnvalue
        (Utilities|Casting|CastToBGA_Athena_SCMachine (.HitActor (Collision|BreakHitResult _outhit))
          (:then
            (return true))
          (:CastFailed
            _)))
      (return false))))

(fn GetCurrentPawnTarget
  (bind _returnvalue (CallFunction|GetAvatarasFortPlayerPawn))
  (bind _returnvalue_2 (AI|GetAIController _returnvalue))
  (bind _returnvalue_3 (AI|GetBlackboard _returnvalue_2))
  (Utilities|IsValid _returnvalue_2
    (:Is%20Valid
      (Utilities|IsValid _returnvalue_3
        (:Is%20Valid
          (bind _asfort_pawn (Utilities|Casting|CastToFortPawn (AI|Components|Blackboard|GetValueasObject _returnvalue_3 (LocalVariables|GetBBKey_TargetActor)))
            (:then
              (return _asfort_pawn))
            (:CastFailed
              _)))
        (:Is%20Not%20Valid
          _)))
    (:Is%20Not%20Valid
      _))
  (return 0))

(fn GetEndForcePullDestination
  (return (+ (Transformation|GetActorLocation (CallFunction|GetAvatarasFortPlayerPawn)) (* (Transformation|GetActorForwardVector (CallFunction|GetAvatarasFortPlayerPawn)) (ScalableFloat|GetValueatLevel (Variables|Default|GetForwardOffsetToPull))))))

(fn CanActivateAbility
  (params (ActorInfo :type Gameplay%20Ability%20Actor%20Info%20Structure) (Handle :type Gameplay%20Ability%20Spec%20Handle%20Structure))
  (bind _fortpawntarget (CallFunction|GetCurrentPawnTarget))
  (bind _returnvalue (Transformation|GetActorLocation _fortpawntarget))
  (bind _returnvalue_2 (CallFunction|GetAvatarasFortPlayerPawn))
  (bind _returnvalue_3 (Actor|GetAttachParentActor _fortpawntarget))
  (bind _returnvalue_4 (Transformation|GetActorLocation _returnvalue_2))
  (bind _tagstovalidateriding (Variables|Default|GetTagsToValidateRiding))
  (bind _returnvalue_5 (GameplayTags|HasAnyMatchingGameplayTags _fortpawntarget _tagstovalidateriding))
  (Utilities|IsValid _fortpawntarget
    (:Is%20Valid
      (if (Math|Float|InRange(Float) (Math|Vector|Distance(Vector) _returnvalue _returnvalue_4) (ScalableFloat|GetValueatLevel (LocalVariables|GetMinDistance)) (ScalableFloat|GetValueatLevel (LocalVariables|GetMaxDistance)))
        (if (WaterBody|IsSurfaceSwimming (CallFunction|GetAvatarasFortPlayerPawn))
          _
          (elif (GameplayTags|HasMatchingGameplayTag _fortpawntarget (Variables|Default|GetKnockbackImmunity))
            _
            (elif _returnvalue_5
              (Utilities|IsValid _returnvalue_3
                (:Is%20Valid
                  (bind _asfort_aipawn (Utilities|Casting|CastToFortAIPawn _returnvalue_3)
                    (:then
                      (bind (_outhit _returnvalue_6) (Collision|LineTraceByChannel _returnvalue _returnvalue_4 "TraceTypeQuery1" false (select _returnvalue_5 (Utilities|Array|MakeArray _fortpawntarget _returnvalue_2 _asfort_aipawn) (Utilities|Array|MakeArray _fortpawntarget _returnvalue_2)) :!label _label))
                      (if _returnvalue_6
                        (Utilities|Casting|CastToBGA_Athena_SCMachine (.HitActor (Collision|BreakHitResult _outhit))
                          (:then
                            _)
                          (:CastFailed
                            (return 0 false))))
                      (return 0 true))
                    (:CastFailed
                      _)))
                (:Is%20Not%20Valid
                  _))
              (else
                (goto _label)))))))
    (:Is%20Not%20Valid
      _))
  (return 0 false))
