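; AI saber-throw ability that throws a projectile at the pawn's blackboard target.
; Activation requires a target within distance range, with an in-water variant requiring both pawns surface swimming.
; Unless visibility is ignored, a line trace ignoring both pawns and any ridden AI mount must be clear, otherwise it commits cooldown and ends.
; The parent's spawn pitch is re-aimed toward the target, and a hold montage plays until the projectile's destruction completes the ability.
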(event Ability|EventActivateAbility
  (bind _returnvalue (CallFunction|GetAvatarasFortPlayerPawn))
  (bind _returnvalue_2 (Actor|GetAttachParentActor (CallFunction|GetCurrentPawnTarget)))
  (bind _fortpawntarget (CallFunction|GetCurrentPawnTarget))
  (bind _actorstoignoreforlinetrace (Variables|Default|GetActorsToIgnoreForLineTrace))
  (Variables|Default|SetCanStopAbility false)
  (if (ScalableFloat|GetValueasBool (Variables|Default|GetIgnoreVisibilityCheck))
    (CallFunction|TriggerThrowfromAI (CallFunction|GetCurrentPawnTarget))
    (else
      (Utilities|Array|Clear _actorstoignoreforlinetrace)
      (Utilities|Array|AddUnique _actorstoignoreforlinetrace _returnvalue)
      (Utilities|Array|AddUnique _actorstoignoreforlinetrace _fortpawntarget)
      (if (GameplayTags|HasAnyMatchingGameplayTags _fortpawntarget (Variables|Default|GetTagsToValidateRiding))
        (Utilities|IsValid _returnvalue_2
          (:Is%20Valid
            (bind _asfort_aipawn (Utilities|Casting|CastToFortAIPawn _returnvalue_2)
              (:then
                (Utilities|Array|AddUnique _actorstoignoreforlinetrace _asfort_aipawn)
                (bind (_outhit _returnvalue_3) (Collision|LineTraceByChannel (Transformation|GetActorLocation _returnvalue) (Transformation|GetActorLocation _fortpawntarget) "TraceTypeQuery1" false (Variables|Default|GetActorsToIgnoreForLineTrace) :!label _label)) ; Do a line check between Actor and player to see if it can reach
                (if (not _returnvalue_3)
                  (CallFunction|TriggerThrowfromAI (CallFunction|GetCurrentPawnTarget))))
              (:CastFailed
                _)))
          (:Is%20Not%20Valid
            _))
        (Ability|CommitAbility) ; If we hit something - start cooldown and try again later
        (Ability|EndAbility))
      (goto _label))))

(event EventAttemptSpawnThrownProjectile
  (bind _projectileref (Variables|ReferencedActors|GetProjectileRef))
  (|CallParentFunction)
  (Utilities|IsValid _projectileref
    (:Is%20Valid
      (Game|BindEventtoOnDestroyed _projectileref (delegate Custom|ProjectileIsDestroyed))
      (Ability|Tasks|PlayMontageAndWaitNotify "None" (Variables|Animation|GetThrowHoldLoopingMontage)))
    (:Is%20Not%20Valid
      (Variables|Default|SetCanStopAbility true))))

(event Custom|ProjectileIsDestroyed
  (params (DestroyedActor :type Actor%20Object%20Reference))
  (Development|PrintString "Hello")
  (Variables|Default|SetCanStopAbility true)
  (Ability|AbilityCompleted))

(event Ability|EventOnEndAbility
  (params (bWasCancelled :type Boolean))
  (|CallParentFunction bWasCancelled) ; This node is disabled and will not be called.\nDrag off pins to build functionality.
)

(fn CalculatePrjSpawnLocationAndRotation
  (bind (_location _rotation) (|CallParentFunction))
  (bind (_rotation_roll _rotation_pitch _rotation_yaw) (Math|Rotator|BreakRotator _rotation))
  (return _location (Math|Rotator|MakeRotator _rotation_roll (select (<= (.z (Transformation|GetActorLocation (CallFunction|GetCurrentPawnTarget))) (.z (Transformation|GetActorLocation (Ability|GetAvatarActorfromActorInfo)))) (* (Class|AbilityFunctions|GetAngleBetweenTwoVectors (Transformation|GetActorForwardVector (Ability|GetAvatarActorfromActorInfo)) (Math|Vector|Normalize (- (Transformation|GetActorLocation (CallFunction|GetCurrentPawnTarget)) (Transformation|GetActorLocation (Ability|GetAvatarActorfromActorInfo))))) -1.0) (Class|AbilityFunctions|GetAngleBetweenTwoVectors (Transformation|GetActorForwardVector (Ability|GetAvatarActorfromActorInfo)) (Math|Vector|Normalize (- (Transformation|GetActorLocation (CallFunction|GetCurrentPawnTarget)) (Transformation|GetActorLocation (Ability|GetAvatarActorfromActorInfo)))))) _rotation_yaw)))

(fn GetCurrentPawnTarget
  (bind _returnvalue (CallFunction|GetAvatarasFortPlayerPawn))
  (bind _returnvalue_2 (AI|GetAIController _returnvalue))
  (bind _returnvalue_3 (AI|GetBlackboard _returnvalue_2))
  (Utilities|IsValid _returnvalue_2
    (:Is%20Valid
      (Utilities|IsValid _returnvalue_3
        (:Is%20Valid
          (bind _asfort_pawn (Utilities|Casting|CastToFortPawn (AI|Components|Blackboard|GetValueasObject _returnvalue_3 (Variables|InternalData|GetBBKey_TargetActor)))
            (:then
              (return _asfort_pawn))
            (:CastFailed
              _)))
        (:Is%20Not%20Valid
          _)))
    (:Is%20Not%20Valid
      _))
  (return 0))

(fn HandleShouldFinishAbility
  (|CallParentFunction)
  (return (Variables|Default|GetCanStopAbility)))

(fn CanActivateAbility
  (params (ActorInfo :type Gameplay%20Ability%20Actor%20Info%20Structure) (Handle :type Gameplay%20Ability%20Spec%20Handle%20Structure))
  (bind _returnvalue (CallFunction|GetAvatarasFortPlayerPawn))
  (bind _fortpawntarget (CallFunction|GetCurrentPawnTarget))
  (bind _returnvalue_2 (WaterBody|IsSurfaceSwimming _fortpawntarget))
  (bind _returnvalue_3 (WaterBody|IsSurfaceSwimming _returnvalue))
  (bind _returnvalue_4 (and _returnvalue_2 _returnvalue_3))
  (Utilities|IsValid _fortpawntarget
    (:Is%20Valid
      (if (Variables|Default|GetInWaterVersion)
        (if _returnvalue_4
          (return 0 (Math|Float|InRange(Float) (Math|Vector|Distance(Vector) (Transformation|GetActorLocation _fortpawntarget) (Transformation|GetActorLocation _returnvalue)) (ScalableFloat|GetValueatLevel (Variables|Default|GetMinDistanceToActivate)) (ScalableFloat|GetValueatLevel (Variables|Default|GetMaxDistanceToActivate))))
          (else
            _))
        (elif _returnvalue_4
          _
          (else
            (return 0 (Math|Float|InRange(Float) (Math|Vector|Distance(Vector) (Transformation|GetActorLocation _fortpawntarget) (Transformation|GetActorLocation _returnvalue)) (ScalableFloat|GetValueatLevel (Variables|Default|GetMinDistanceToActivate)) (ScalableFloat|GetValueatLevel (Variables|Default|GetMaxDistanceToActivate))))))))
    (:Is%20Not%20Valid
      _))
  (return 0 false))
