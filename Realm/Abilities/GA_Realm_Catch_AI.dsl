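; Gameplay ability blueprint GA_Realm_Catch_AI, by its name the AI version of a catch ability.
; Activation and end events only forward to the parent class, which holds the actual behaviour.
; A montage-completed handler ends the ability through a DoOnce, so repeated completion calls end it only once.
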(event Ability|EventActivateAbility
  (|CallParentFunction))

(event EventHandleMontageCompleted
  (Utilities|FlowControl|DoOnce false
    (:Completed
      (Ability|EndAbility))))

(event Ability|EventOnEndAbility
  (params (bWasCancelled :type Boolean))
  (|CallParentFunction bWasCancelled))
