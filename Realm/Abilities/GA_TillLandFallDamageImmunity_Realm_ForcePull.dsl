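; Gameplay ability for Realm's Force Pull, a child of what appears to be a till-land fall damage immunity ability, that adds pull animation and weapon holstering.
; On activation it runs parent logic, then, unless already ended or the pawn has riding tags, holsters the weapon and plays the Force Pull montage.
; On end it clears the movement-mode validation timer and, for valid non-riding pawns, stops the montage and unholsters.
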(event Ability|EventActivateAbility
  (bind _forcepullmontage (Variables|Default|GetForcePullMontage))
  (|CallParentFunction)
  (bind _asfort_player_pawn_athena (Utilities|Casting|CastToFortPlayerPawnAthena (Ability|GetAvatarActorfromActorInfo)))
  (Variables|InternalData|SetPlayerPawn _asfort_player_pawn_athena)
  (if (not (or (Variables|InternalData|GetHasAbilityEnded) (GameplayTags|HasAnyMatchingGameplayTags (Variables|InternalData|GetPlayerPawn) (Variables|Default|GetTagsToValidateRiding))))
    (Weapon|SetHolsterWeapon (Variables|InternalData|GetPlayerPawn) 0 true false (Variables|Default|GetHolsterId))
    (Animation|PlayAnimMontage _asfort_player_pawn_athena _forcepullmontage 1.0 "None")
    (Variables|InternalData|SetCurrentlyPlayingMontage _forcepullmontage)))

(event Ability|EventOnEndAbility
  (params (bWasCancelled :type Boolean))
  (bind _playerpawn (Variables|InternalData|GetPlayerPawn))
  (Variables|InternalData|SetHasAbilityEnded true)
  (Utilities|Time|ClearandInvalidateTimerbyHandle (Variables|InternalData|GetMovementModeValidationTimerHandle))
  (|CallParentFunction bWasCancelled)
  (Utilities|IsValid _playerpawn
    (:Is%20Valid
      (if (not (GameplayTags|HasAnyMatchingGameplayTags _playerpawn (Variables|Default|GetTagsToValidateRiding)))
        (Animation|StopAnimMontage _playerpawn (Variables|InternalData|GetCurrentlyPlayingMontage))
        (Weapon|SetHolsterWeapon _playerpawn 0 false false (Variables|Default|GetHolsterId))
        (Utilities|FlowControl|Delay 0.2)))))

(event Ability|EventFailedToActivatePassiveAbility
  (params (ActorInfo :type Gameplay%20Ability%20Actor%20Info%20Structure))
  (|CallParentFunction))
